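# Unlit translucent mesh-particle material for the nectar aura effect of the butterfly pet.
# Two panning smoke samples are blended, with tiling and speed from dynamic parameters, then shaped by a V gradient that fades toward high V and boosts low V.
# That mask blends dark to light color and, with particle alpha, forms opacity; particle color tints the result.
# Depth fade and HDR remapping apply on the default shading path; mobile skips both.
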Material /Figment_S06/Athena/Effects/Particles/Pets/Butterfly/M_Butterfly_Nectar_Aura
graph {
 properties {bUsedWithMeshParticles=true, blendMode=BLEND_Translucent, pixelDepthOffsetMode=PDOM_Legacy, shadingModel=MSM_Unlit, usedShadingModels=Unlit}
 nodes [
  n1 = ParticleColor()
  n2 = OneMinus(None=ComponentMask(None=TextureCoordinate()) {g=true})
  n3 = Multiply(A=Power(Base=@n2) {constExponent=6}) {constB=0.5}
  n4 = DynamicParameter() {defaultValue=<rgba 1 -1 1 1>, paramNames [YTile,PanSpeed,Param3,Param4]}
  n5 = AppendVector(A=Constant() {r=1}, B=@n4.YTile)
  n6 = ScalarParameter() {defaultValue=2, parameterName=UScale}
  n7 = ScalarParameter() {defaultValue=0.5, parameterName=VScale}
  n8 = Multiply(A=@n5, B=Multiply(A=TextureCoordinate(), B=AppendVector(A=@n6, B=@n7)))
  n9 = Multiply(A=Time(), B=@n4.PanSpeed)
  n10 = ScalarParameter() {defaultValue=5, parameterName=UScale_1}
  n11 = ScalarParameter() {defaultValue=0.2, parameterName=VScale_1}
  n12 = Multiply(A=@n5, B=Multiply(A=TextureCoordinate(), B=AppendVector(A=@n10, B=@n11)))
  n13 = LinearInterpolate(
   A=TextureSample(UVs=Panner(Coordinate=@n8, Time=@n9) {speedX=0.2, speedY=1}) {texture=/Game/Effects/Fort_Effects/Textures/Tile/T_Inky_Smoke_Tile_Chromatic}.RGB
   B=TextureSample(UVs=Panner(Coordinate=@n12, Time=@n9) {speedX=-0.25, speedY=0.8}) {texture=&T_Inky_Smoke_Tile_Chromatic}.RGB
  )
  n14 = Multiply(A=@n2, B=Add(A=Multiply(A=@n13) {constB=1.5}, B=@n3))
  n15 = ScalarParameter() {defaultValue=50, parameterName=DepthFadeDist}
  Multiply(
   A=Multiply(A=@n1.A, B=Add(A=@n3, B=@n14))
   B=ShadingPathSwitch(Default=DepthFade(FadeDistance=@n15), Mobile=Constant() {r=1}) {desc=Migrated from FeatureLevelSwitch}
  )
  n16 = Multiply(
   A=LinearInterpolate(
    A=VectorParameter() {parameterName=DarkColor, defaultValue=<rgba 0.02 0.02 0.02 0>}.RGB
    B=VectorParameter() {parameterName=LightColor, defaultValue=<rgba 1 1 1 0>}.RGB
    Alpha=@n14
   )
   B=@n1.RGB
  )
  n17 = Clamp(None=@n16)
  n18 = ScalarParameter() {defaultValue=1.5, parameterName=HDRMult}
  ShadingPathSwitch(
   Default=MaterialFunctionCall(LDR Input=@n17, HDR Tint=Multiply(A=@n16, B=@n18)) {materialFunction=&DeriveHDRfromLDR}.RemappedRange
   Mobile=@n17
  ) {desc=Migrated from FeatureLevelSwitch}
 ]
}
