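# Translucent lit master material for glowing, flowing nectar on the Butterfly pet's mesh particles.
# Two panning noises distort a liquid mask that drives color and optional sweeping emissive, pulsed by MPC_Cube's Heartbeat.
# Switchable vertex offsets add normal-map wobble, vertical stretch, modulated motion, spline thickening and Nectar-driven collapse.
# Dynamic parameters feed stretch, spline width, emissive and spline-only opacity shaping, with depth fade softening intersections.
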Material /Figment_S06/Athena/Effects/Particles/Pets/Butterfly/M_Butterfly_Nectar_Master
graph {
 properties {
  bUsedWithMeshParticles=true
  blendMode=BLEND_Translucent
  pixelDepthOffsetMode=PDOM_Legacy
  translucencyLightingMode=TLM_VolumetricPerVertexDirectional
  usedShadingModels=Default Lit
 }
 nodes [
  n1 = ScalarParameter() {defaultValue=0.0736, group=Emissive, parameterName=Emissive Fresnel Exp, sortPriority=1}
  n2 = VectorParameter() {group=Emissive, parameterName=Color_Emissive, sortPriority=2, defaultValue=<rgba 4 0.8 3.9368 1>}
  n3 = Multiply(A=OneMinus(None=Fresnel(ExponentIn=@n1)), B=@n2.RGB)
  n4 = StaticBoolParameter() {defaultValue=true, group=Emissive, parameterName=Use Fresnel for Emissive}
  n5 = StaticSwitch(True=@n3, False=@n2.RGB, Value=@n4)
  n6 = Clamp(None=@n5)
  n7 = Multiply(
   A=ShadingPathSwitch(
    Default=MaterialFunctionCall(LDR Input=@n6, HDR Tint=@n5) {materialFunction=&DeriveHDRfromLDR}.RemappedRange
    Mobile=StaticSwitch(True=@n3, False=@n2.RGB, Value=@n4)
   ) {desc=Migrated from FeatureLevelSwitch}
   B=ScalarParameter() {defaultValue=0.25, group=Emissive, parameterName=Overall Emissive Intensity}
  )
  n8 = Panner() {speedY=-0.1}
  n9 = Multiply(
   A=TextureCoordinate()
   B=ComponentMask(None=VectorParameter() {group=Liquid, parameterName=Liquid_UV_Scale, defaultValue=<rgba 2 2 0 0>}.RGB) {g=true, r=true}
  )
  n10 = Multiply(
   A=ComponentMask(None=VectorParameter() {group=Liquid, parameterName=Distortion_A_UV_Scale, defaultValue=<rgba 3 3 0 0>}.RGB) {g=true, r=true}
   B=TextureCoordinate()
  )
  n11 = Constant() {r=0.05}
  n12 = Multiply(A=Time(), B=@n11)
  n13 = Multiply(
   A=ComponentMask(None=VectorParameter() {group=Liquid, parameterName=Distortion_B_UV_Scale, defaultValue=<rgba 1 1 0 0>}.RGB) {g=true, r=true}
   B=TextureCoordinate()
  )
  n14 = Multiply(A=Time(), B=@n11)
  n15 = Multiply(
   A=TextureSample(UVs=Panner(Coordinate=@n10, Time=@n12) {speedY=-1}) {samplerType=SAMPLERTYPE_Grayscale, texture=&LowResBlurredNoise}.R
   B=TextureSample(UVs=Panner(Coordinate=@n13, Time=@n14) {speedX=0.5, speedY=-1}) {samplerType=SAMPLERTYPE_Grayscale, texture=&LowResBlurredNoise}.R
  )
  n16 = TextureSample(UVs=Add(A=@n9, B=@n15)) {
   samplerSource=SSM_Wrap_WorldGroupSettings
   samplerType=SAMPLERTYPE_Masks
   texture=/Figment_S06/Athena/Prototype/Blueprints/Cube/Materials/T_CubeNoise
  }
  n17 = ScalarParameter() {defaultValue=-0.3, parameterName=dasf34t_5}
  n18 = Power(Base=@n15, Exp=ScalarParameter() {defaultValue=0.75, group=Liquid, parameterName=Distortion Contrast})
  n19 = ScalarParameter() {defaultValue=5, group=Liquid, parameterName=Distortion Intensity}
  n20 = Multiply(A=Saturate(None=Add(A=@n16.G, B=@n17)), B=Saturate(None=Multiply(A=@n18, B=@n19)))
  n21 = LinearInterpolate(
   B=ScalarParameter() {defaultValue=20, group=Emissive, parameterName=Emissive_Intensity_Sweep}
   Alpha=Multiply(
    A=TextureSample(UVs=@n8) {samplerSource=SSM_Wrap_WorldGroupSettings, texture=/Game/Effects/Fort_Effects/Textures/Masks/T_Vert_Gradient_Softer}.R
    B=@n20
   )
  ) {constA=1, constB=2}
  n22 = ScalarParameter() {defaultValue=40, group=Emissive, parameterName=End Emissive}
  n23 = TextureCoordinate()
  n24 = Multiply(A=@n20, B=Power(Base=ComponentMask(None=@n23) {g=true}, Exp=Constant() {r=8.3256}))
  n25 = StaticSwitchParameter(True=Multiply(A=Multiply(A=@n7, B=@n21), B=LinearInterpolate(B=@n22, Alpha=@n24) {constA=1}), False=@n7) {defaultValue=true, group=Emissive, parameterName=Use Sweeping Emissive}
  n26 = DynamicParameter() {defaultValue=<rgba 0 1 1 1>, paramNames [Stretch, Opacity, Spline Scale, Drive Emissive]}
  n27 = StaticSwitchParameter(True=Multiply(A=@n25, B=Multiply(A=@n26.Drive Emissive, B=@n20)), False=@n25) {group=Emissive, parameterName=Allow Dynamic Para to drive Emissive}
  LinearInterpolate(
   A=@n27
   B=Multiply(A=@n27, B=ScalarParameter() {defaultValue=10, group=Emissive, parameterName=Heartbeat Intensity}) {constB=10}
   Alpha=CollectionParameter() {collection=/Game/Athena/Prototype/Blueprints/Cube/Materials/MPC_Cube, parameterName=Heartbeat}
  )
  LinearInterpolate(
   A=VectorParameter() {group=Base Color, parameterName=Color Low, defaultValue=<rgba 0.0804 0 0.2 1>}.RGB
   B=VectorParameter() {group=Base Color, parameterName=Color High, sortPriority=1, defaultValue=<rgba 2 0.1 1.9625 1>}.RGB
   Alpha=@n20
  )
  n28 = ComponentMask(None=TextureCoordinate() {coordinateIndex=1}) {g=true}
  n29 = LinearInterpolate(
   A=ScalarParameter() {defaultValue=-1, parameterName=A}
   B=ScalarParameter() {defaultValue=0.2, parameterName=B}
   Alpha=CollectionParameter() {collection=&MPC_Cube, parameterName=Nectar}
  ) {constA=-0.5, constAlpha=0}
  Saturate(None=Multiply(A=Add(A=@n28, B=@n29)) {constB=10})
  n30 = TextureSample(UVs=Panner(Coordinate=@n9) {speedY=-0.05}) {
   samplerSource=SSM_Wrap_WorldGroupSettings
   samplerType=SAMPLERTYPE_Normal
   texture=/Game/VisualThreatManager/StormVisuals/Textures/2ndInfectedMapNormal
  }
  n31 = MaterialFunctionCall(Normal=@n30.RGB, Flatness=ScalarParameter() {defaultValue=0.95, group=Normal, parameterName=Flatness}) {materialFunction=&FlattenNormal}
  n32 = Add(A=ComponentMask(None=@n31.Result) {r=true}, B=ComponentMask(None=@n31.Result) {g=true})
  n33 = Multiply(
   A=Saturate(None=Multiply(A=@n32) {constB=10})
   B=Multiply(A=VertexNormalWS(), B=ScalarParameter() {group=WPO, parameterName=General WPO Offset})
  )
  n34 = ScalarParameter() {defaultValue=-750, group=WPO, parameterName=Stretch_Height, sortPriority=2}
  n35 = Add(
   A=DynamicParameter() {defaultValue=<rgba 0 1 1 1>, paramNames [Stretch, Opacity, Spline Scale, Drive Emissive]}.Stretch
   B=ScalarParameter() {group=WPO, parameterName=Bottom Stretch, sortPriority=4}
  )
  n36 = LinearInterpolate(
   A=Constant3Vector()
   B=AppendVector(A=Constant2Vector(), B=LinearInterpolate(B=@n34, Alpha=@n35))
   Alpha=ComponentMask(None=TextureCoordinate()) {g=true}
  )
  n37 = ScalarParameter() {group=WPO, parameterName=Bottom Movement, sortPriority=5}
  n38 = AppendVector(
   A=Constant2Vector()
   B=LinearInterpolate(
    B=ScalarParameter() {defaultValue=-750, group=WPO, parameterName=Drip_Height, sortPriority=3}
    Alpha=LinearInterpolate(Alpha=@n37) {constA=-1, constB=0}
   )
  )
  n39 = StaticSwitchParameter(True=Add(A=@n33, B=Add(A=@n36, B=@n38)), False=@n33) {group=WPO, parameterName=Use Vertical Stretch, sortPriority=1}
  n40 = ScalarParameter() {defaultValue=-250, group=WPO, parameterName=Mod_ZHeight_Large, sortPriority=8}
  n41 = ScalarParameter() {defaultValue=-150, group=WPO, parameterName=Mod_ZHeight_Small, sortPriority=7}
  n42 = ComponentMask(None=TextureCoordinate()) {g=true}
  n43 = Panner() {speedY=-0.1}
  n44 = AppendVector(
   A=Constant2Vector()
   B=LinearInterpolate(
    B=LinearInterpolate(A=@n40, B=@n41, Alpha=@n42)
    Alpha=TextureSample(UVs=@n43) {samplerSource=SSM_Wrap_WorldGroupSettings, texture=&T_Vert_Gradient_Softer}.R
   )
  )
  n45 = VertexColor()
  n46 = StaticSwitchParameter(True=Add(A=@n39, B=LinearInterpolate(B=@n44, Alpha=@n45.R)), False=@n39) {group=WPO, parameterName=Use Modulated Vert Movement, sortPriority=6}
  n47 = DynamicParameter() {defaultValue=<rgba 0 1 1 1>, paramNames [Stretch, Opacity, Spline Scale, Drive Emissive]}
  n48 = Multiply(A=ScalarParameter() {defaultValue=2, group=WPO, parameterName=Width Base}, B=@n47.Spline Scale)
  n49 = Multiply(A=ScalarParameter() {defaultValue=2, group=WPO, parameterName=Width Tip}, B=@n47.Spline Scale)
  n50 = StaticSwitchParameter(
   True=Add(
    A=@n46
    B=MaterialFunctionCall(WidthBase=@n48, WidthTip=@n49) {materialFunction=&SplineThicken}.WorldPositionOffset
   )
   False=@n46
  ) {group=WPO, parameterName=Is Spline Mesh, sortPriority=6}
  n51 = OneMinus(None=Subtract(A=WorldPosition().XYZ, B=ObjectPositionWS()))
  n52 = ComponentMask(None=TextureCoordinate() {coordinateIndex=1}) {g=true}
  n53 = LinearInterpolate(
   A=ScalarParameter() {defaultValue=-0.75, parameterName=A_2}
   B=ScalarParameter() {defaultValue=2, parameterName=B_2}
   Alpha=CollectionParameter() {collection=&MPC_Cube, parameterName=Nectar}
  ) {constA=-0.5, constAlpha=0}
  n54 = Clamp(None=Multiply(A=Add(A=@n52, B=@n53)) {constB=2})
  StaticSwitchParameter(True=Add(A=@n50, B=LinearInterpolate(A=@n51, Alpha=@n54) {constB=0}), False=@n50) {group=WPO, parameterName=Nectar Controlled Collapse}
  n55 = TextureCoordinate()
  n56 = Power(
   Base=Sine(None=ComponentMask(None=@n55) {r=true}) {period=2}
   Exp=ScalarParameter() {defaultValue=4, group=Opacity, parameterName=Opacity Contrast}
  )
  n57 = TextureCoordinate() {uTiling=1.2}
  n58 = Constant2Vector() {g=0.95, r=0.6}
  n59 = Saturate(None=Multiply(A=SphereMask(A=@n57, B=@n58) {attenuationRadius=0.95, hardnessPercent=0}) {constB=15})
  n60 = ComponentMask(None=TextureCoordinate()) {g=true}
  n61 = Multiply(
   A=Multiply(A=Multiply(A=@n56, B=@n59), B=ParticleColor().A)
   B=Power(
    Base=OneMinus(None=@n60)
    Exp=DynamicParameter() {defaultValue=<rgba 0 1 1 1>, paramNames [Stretch, Opacity, Spline Scale, Drive Emissive]}.Opacity
   )
  )
  n62 = Constant() {r=1}
  n63 = DepthFade(
   Opacity=Multiply(
    A=ScalarParameter() {defaultValue=0.95, group=Opacity, parameterName=Opacity, sortPriority=1}
    B=StaticSwitchParameter(True=@n61, False=@n62) {group=WPO, parameterName=Is Spline Mesh, sortPriority=6}
   )
   FadeDistance=ScalarParameter() {defaultValue=250, group=Opacity, parameterName=Depth Fade Distance, sortPriority=2}
  )
  StaticSwitchParameter(True=Multiply(A=@n63, B=Constant() {r=1}), False=@n63) {parameterName=Nectar Controlled Collapse}
 ]
}
